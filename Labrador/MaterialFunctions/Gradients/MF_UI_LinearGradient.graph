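# UI material function, exposed to the library, that outputs a grayscale linear gradient with adjustable start, end, rotation and offset.
# UVs, defaulting to normalized UI UVs, pass through CustomRotator, have GradientOffset subtracted and are masked to red.
# That ramp is only the default of the Texture input, so a supplied grayscale value replaces rotation and offset.
# The value is remapped from GradientStart–GradientEnd to 0–1 and saturated.
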MaterialFunction /Labrador/MaterialFunctions/Gradients/MF_UI_LinearGradient
graph {
 properties {bExposeToLibrary=true, description=Linear Gradient Function, libraryCategoriesText [UI_ShaderLibrary]}
 nodes [
  n1 = FunctionInput() {
   bUsePreviewValueAsDefault=true
   description=Rotation pivot point coordinates
   inputName=RotationCenter
   inputType=FunctionInput_Vector2
   sortPriority=5
   previewValue=<0.5 0.5 0.5 1>
  }
  n2 = FunctionInput() {
   bUsePreviewValueAsDefault=true
   description=The rotation angle of the gradient from 0-1
   inputName=RotationAngle
   inputType=FunctionInput_Scalar
   sortPriority=4
  }
  n3 = MaterialFunctionCall(
   UVs=FunctionInput(Preview=MaterialFunctionCall() {materialFunction=&GetUserInterfaceUV}.Normalized UV) {bUsePreviewValueAsDefault=true, inputName=UVs, inputType=FunctionInput_Vector2}
   Rotation Center=@n1
   "Rotation Angle (0-1)"=@n2
  ) {materialFunction=&CustomRotator}
  n4 = FunctionInput() {
   bUsePreviewValueAsDefault=true
   description=Gradient offset that can be positive or negative values
   inputName=GradientOffset
   inputType=FunctionInput_Scalar
   sortPriority=6
  }
  n5 = FunctionInput(Preview=ComponentMask(None=Subtract(A=@n3.Rotated Values, B=@n4)) {r=true}) {
   bUsePreviewValueAsDefault=true
   description=Grayscale Texture for Gradient remapping
   inputName=Texture
   inputType=FunctionInput_Scalar
   sortPriority=1
  }
  n6 = FunctionInput() {
   bUsePreviewValueAsDefault=true
   description=Gradient start point from 0-1
   inputName=GradientStart
   inputType=FunctionInput_Scalar
   sortPriority=2
  }
  n7 = Subtract(A=@n5, B=@n6)
  n8 = FunctionInput() {
   bUsePreviewValueAsDefault=true
   description=Gradient end point from 0-1
   inputName=GradientEnd
   inputType=FunctionInput_Scalar
   sortPriority=3
   previewValue=<1 1 1 1>
  }
  n9 = Subtract(A=@n8, B=@n6)
  FunctionOutput(None=Saturate(None=Divide(A=@n7, B=@n9)))
 ]
}
